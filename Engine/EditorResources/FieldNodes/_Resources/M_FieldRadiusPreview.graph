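# Translucent lit editor material that, by its name and path, previews the radius of Field Nodes.
# Preview_Colour scaled by AlphaMax blends with a PulseLevel-scaled copy on a time sine, with Active fading that pulse in; the default PulseLevel leaves it static.
# A Fresnel term, its exponent growing with object scale, blends AlphaMin to AlphaMax for a view-dependent rim value.
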Material /Engine/EditorResources/FieldNodes/_Resources/M_FieldRadiusPreview
graph {
 properties {blendMode=BLEND_Translucent, pixelDepthOffsetMode=PDOM_Legacy, usedShadingModels=Default Lit}
 nodes [
  n1 = ScalarParameter() {defaultValue=1, parameterName=PulseLevel}
  n2 = ScalarParameter() {defaultValue=1, group=Pattern, parameterName=AlphaMax}
  n3 = VectorParameter() {parameterName=Preview_Colour, defaultValue=<rgba 0 0.4363 1 1>}
  n4 = Multiply(A=@n2, B=@n3.RGB)
  n5 = Time()
  LinearInterpolate(
   A=@n4
   B=LinearInterpolate(A=Multiply(A=@n4, B=@n1), B=@n4, Alpha=Sine(None=@n5))
   Alpha=ScalarParameter() {defaultValue=1, parameterName=Active}
  )
  n6 = ScalarParameter() {defaultValue=0.02, group=Pattern, parameterName=AlphaMin}
  Multiply(A=@n3.RGB, B=@n6)
  n7 = Constant() {r=1}
  n8 = Add(
   A=@n7
   B=Divide(A=MaterialFunctionCall() {materialFunction=/Engine/Functions/Engine_MaterialFunctions02/WorldPositionOffset/ObjectScale}.Scale XYZ, B=@n7) {constB=5}
  )
  Reroute(None=LinearInterpolate(A=@n6, B=@n2, Alpha=Fresnel(ExponentIn=@n8) {baseReflectFraction=0.01}))
 ]
}
