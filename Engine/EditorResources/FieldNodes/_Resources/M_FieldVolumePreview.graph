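# Translucent lit editor material that, going by its name and folder, previews field node volumes in a tintable colour.
# Blends a dimmed and full Preview_Colour by a mask lerped from AlphaMin to AlphaMax using colour times Power; no texture is sampled.
# A sine of Time and PulseSpeed swings between normal and inverse blends unless the default-on "No Pulse?" static switch bypasses it.
# The Active parameter lerps two final blends toward that result.
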Material /Engine/EditorResources/FieldNodes/_Resources/M_FieldVolumePreview
graph {
 properties {bUsedWithStaticLighting=true, blendMode=BLEND_Translucent, pixelDepthOffsetMode=PDOM_Legacy, usedShadingModels=Default Lit}
 nodes [
  n1 = VectorParameter() {group=Activation, parameterName=Preview_Colour, sortPriority=1, defaultValue=<rgba 1 0 0.0794 1>}
  n2 = Multiply(A=@n1.RGB) {constB=0.1}
  n3 = LinearInterpolate(
   A=ScalarParameter() {defaultValue=0.3, group=Pattern, parameterName=AlphaMin, sortPriority=1}
   B=ScalarParameter() {defaultValue=0.6, group=Pattern, parameterName=AlphaMax, sortPriority=2}
   Alpha=Multiply(
    A=ScalarParameter() {defaultValue=1, group=Activation, parameterName=Power, sortPriority=2}
    B=VectorParameter() {group=Activation, parameterName=Preview_Colour, sortPriority=1, defaultValue=<rgba 1 0 0.0794 1>}.RGB
   )
  ) {constA=0.25, constB=0.75}
  n4 = LinearInterpolate(A=@n2, B=@n1.RGB, Alpha=@n3) {constA=0.2, desc=Normal Pattern}
  n5 = ScalarParameter() {defaultValue=1, group=Activation, parameterName=PulseSpeed, sortPriority=4}
  n6 = Time()
  n7 = LinearInterpolate(
   A=@n4
   B=LinearInterpolate(A=@n2, B=@n1.RGB, Alpha=OneMinus(None=@n3)) {constA=0.2, desc=Inverse Pattern}
   Alpha=Sine(None=Multiply(A=@n5, B=@n6))
  ) {desc=Transition between Normal and Inverse Pattern}
  n8 = ScalarParameter() {defaultValue=1, group=Activation, parameterName=Active}
  LinearInterpolate(
   A=@n4
   B=StaticSwitchParameter(True=@n4, False=@n7) {defaultValue=true, group=Activation, parameterName=No Pulse?, sortPriority=3}
   Alpha=@n8
  )
  LinearInterpolate(
   A=LinearInterpolate(Alpha=@n3)
   B=StaticSwitchParameter(True=@n4, False=@n7) {defaultValue=true, group=Activation, parameterName=No Pulse?, sortPriority=3}
   Alpha=@n8
  )
 ]
}
